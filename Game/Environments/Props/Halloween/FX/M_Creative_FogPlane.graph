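# Translucent, lit fog plane for Halloween props, with per-instance color, opacity, tiling and scroll controls through custom primitive data.
# Noise-distorted, scale-compensated panning UVs, optionally radial, drive three blended smoke layers for color and opacity.
# Sphere mask, optional Fresnel, vertex green and depth/near-camera fades shape opacity, with fades skipped on mobile and low quality.
# Blended cloud normals modulate color through view-space reflection, while panned textures drive masked vertex offset.
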Material /Game/Environments/Props/Halloween/FX/M_Creative_FogPlane
graph {
 properties {
  allowTranslucentCustomDepthWrites=true
  blendMode=BLEND_Translucent
  opacityMaskClipValue=0.05
  pixelDepthOffsetMode=PDOM_Legacy
  translucencyDirectionalLightingIntensity=0.1
  translucencyLightingMode=TLM_VolumetricPerVertexDirectional
  usedShadingModels=Default Lit
 }
 nodes [
  n1 = MaterialFunctionCall(Vector or UVs=TextureCoordinate()) {materialFunction=&VectorToRadialValue}
  n2 = TextureCoordinate()
  n3 = MaterialFunctionCall() {materialFunction=/Engine/Functions/Engine_MaterialFunctions02/WorldPositionOffset/ObjectScale}
  n4 = Multiply(
   A=Multiply(
    A=StaticSwitchParameter(True=@n1.Radial Coordinates, False=@n2) {parameterName=UseRadialUVs}
    B=ScalarParameter() {bUseCustomPrimitiveData=true, defaultValue=1, parameterName=TileFrequency, primitiveDataIndex=9}
   )
   B=AppendVector(A=@n3.Scale X, B=@n3.Scale X)
  )
  n5 = ScalarParameter() {bUseCustomPrimitiveData=true, defaultValue=1, parameterName=PannerTimeRandom, primitiveDataIndex=11}
  n6 = Time()
  n7 = Multiply(
   A=Multiply(A=Add(A=@n5, B=@n6), B=ScalarParameter() {bUseCustomPrimitiveData=true, defaultValue=1, parameterName=TileFrequency, primitiveDataIndex=9})
   B=ScalarParameter() {bUseCustomPrimitiveData=true, defaultValue=1, parameterName=PannerSpeed, primitiveDataIndex=10}
  )
  n8 = Panner(Coordinate=@n4, Time=@n7) {speedY=0.03}
  n9 = Divide(A=ComponentMask(None=TextureSample(UVs=@n8) {texture=&T_NoisePack_01}.RGB) {g=true, r=true}) {constB=15}
  n10 = Add(A=Panner(Coordinate=@n4, Time=@n7) {speedY=-0.1}, B=@n9)
  n11 = ComponentMask(None=VectorParameter() {parameterName=TextureB_UVtiling, defaultValue=<rgba 3 1 0 0>}.RGB) {g=true, r=true}
  n12 = Add(A=Panner(Coordinate=Multiply(A=@n4, B=@n11), Time=@n7) {speedY=-0.1}, B=@n9)
  n13 = ComponentMask(None=VectorParameter() {parameterName=TextureC_UVtiling, defaultValue=<rgba 2 1 0 0>}.RGB) {g=true, r=true}
  n14 = Add(A=Panner(Coordinate=Multiply(A=@n4, B=@n13), Time=@n7) {speedY=-0.1}, B=@n9)
  n15 = MaterialFunctionCall(
   A=TextureSample(UVs=@n10) {desc=A, samplerType=SAMPLERTYPE_LinearColor, texture=/Game/Global/Textures/Default/Tileable_Patterns/T_Tile_Smoke_Small}.R
   B=TextureSample(UVs=@n12) {desc=B, samplerType=SAMPLERTYPE_LinearColor, texture=&T_Tile_Smoke_Small}.R
   C=TextureSample(UVs=@n14) {desc=C, texture=/Game/Global/Textures/Default/Tileable_Patterns/T_Smoke_Puff_Tile_Up}.R
   Alpha=MaterialFunctionCall() {materialFunction=&LinearGradient}.VGradient
  ) {materialFunction=&3ColorBlend}
  n16 = ScalarParameter() {defaultValue=0.5, parameterName=VertexPixelNormalLerp}
  n17 = Saturate(None=TwoSidedSign())
  n18 = MaterialFunctionCall(
   Brightness=ScalarParameter() {defaultValue=1, parameterName=FresnelBrightness}
   Exponent=ScalarParameter() {defaultValue=1, parameterName=FresnelExponent}
   Vertex / Pixel Normal=LinearInterpolate(A=PixelNormalWS(), B=VertexNormalWS(), Alpha=Multiply(A=@n16, B=@n17))
  ) {materialFunction=&MF_Fresnel_FX}
  n19 = SphereMask(
   A=TextureCoordinate()
   B=Constant() {r=0.5}
   Radius=ScalarParameter() {defaultValue=0.5, parameterName=MaskRadius}
   Hardness=ScalarParameter() {parameterName=MaskHardness}
  )
  n20 = Multiply(A=StaticSwitchParameter(True=Multiply(A=@n15.Result, B=@n18.Fresnel), False=@n15.Result) {parameterName=UseFresnel}, B=@n19)
  n21 = ScalarParameter() {defaultValue=50, parameterName=FadeDistance}
  n22 = ScalarParameter() {defaultValue=80, parameterName=NearCameraFadeDistance}
  n23 = ShadingPathSwitch(
   Default=Multiply(
    A=DepthFade(Opacity=@n20, FadeDistance=@n21) {fadeDistanceDefault=50}
    B=MaterialFunctionCall(Distance=@n22) {materialFunction=&Near_Camera_Fade}.Result
   )
   Mobile=@n20
  ) {desc=Migrated from FeatureLevelSwitch}
  Multiply(
   A=Multiply(
    A=ScalarParameter() {bUseCustomPrimitiveData=true, defaultValue=1, group=FogFading, parameterName=FogOpacity, primitiveDataIndex=8}
    B=QualitySwitch(Default=@n23, Low=@n20, High=@n23, Medium=@n23)
   )
   B=VertexColor().G
  )
  n24 = VectorParameter() {bUseCustomPrimitiveData=true, parameterName=ColorLow, defaultValue=<rgba 0.4884 0.4741 0.5104 1>}
  n25 = VectorParameter() {bUseCustomPrimitiveData=true, parameterName=ColorHigh, primitiveDataIndex=4, defaultValue=<rgba 0.7747 0.8735 1 1>}
  n26 = ComponentMask(None=Panner(Time=@n7) {speedY=-0.1}) {g=true}
  n27 = MaterialFunctionCall(
   A=TextureSample(UVs=@n10) {samplerType=SAMPLERTYPE_Normal, texture=/Game/Effects/Fort_Effects/Creatures/Elemental/Ice/TilingCloudNormalsMap}.RGB
   B=TextureSample(UVs=@n12) {samplerType=SAMPLERTYPE_Normal, texture=&TilingCloudNormalsMap}.RGB
   C=TextureSample(UVs=@n14) {samplerType=SAMPLERTYPE_Normal, texture=&TilingCloudNormalsMap}.RGB
   Alpha=ConstantBiasScale(None=Sine(None=@n26) {period=2})
  ) {materialFunction=&3ColorBlend}
  n28 = ComponentMask(None=Transform(None=ReflectionVectorWS(CustomWorldNormal=@n27.Result)) {transformType=TRANSFORM_View}) {b=true}
  n29 = Constant()
  Multiply(
   A=Multiply(
    A=LinearInterpolate(A=@n24.RGB, B=@n25.RGB, Alpha=@n15.Result)
    B=ShadingPathSwitch(Default=@n28, Mobile=@n29) {desc=Migrated from FeatureLevelSwitch}
   ) {constB=0.2}
   B=ScalarParameter() {defaultValue=0.3, parameterName=Emissive}
  ) {constB=0.25}
  n30 = VectorParameter() {group=WPO, parameterName=TextureWPO_UVtiling, defaultValue=<rgba 4 2 0 0>}
  n31 = Panner(Coordinate=Multiply(A=@n4, B=ComponentMask(None=@n30.RGB) {g=true, r=true}), Time=@n7) {speedY=-0.15}
  n32 = TextureSample(UVs=@n31) {samplerType=SAMPLERTYPE_Normal, texture=&T_Weap_Distort_N_02}
  n33 = ScalarParameter() {defaultValue=25, group=WPO, parameterName=WPO_Height}
  n34 = MaterialFunctionCall() {materialFunction=&ObjectScale}
  n35 = AppendVector(
   A=Multiply(A=ComponentMask(None=@n32.RGB) {g=true, r=true}) {constB=0.5}
   B=Multiply(A=TextureSample(UVs=@n31) {texture=&T_Smooth_tile}.R, B=Multiply(A=@n33, B=@n34.Scale Z)) {constB=0.6}
  )
  Reroute(None=Multiply(A=ConstantBiasScale(None=@n19) {bias=0.5, scale=0.6667}, B=Transform(None=@n35)))
 ]
}
