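# Lit translucent, two-sided mesh-particle material for a Halloween fog machine's fog plane.
# Three panning smoke layers, noise-distorted and blended along V, form opacity, shaped by gradient, vertex color, particle alpha, depth fade and switchable erosion.
# Matching blended normals drive a view-space reflection highlight tinted between two colors for emissive.
# Panned distortion textures animate world position offset, replaced by a constant on Low and mobile, which also skip depth fade.
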Material /Game/Environments/Props/Halloween/FX/M_FogMachine_FogPlane
graph {
 properties {
  allowTranslucentCustomDepthWrites=true
  bUsedWithMeshParticles=true
  blendMode=BLEND_Translucent
  opacityMaskClipValue=0.05
  pixelDepthOffsetMode=PDOM_Legacy
  previewMesh=/Game/VisualThreatManager/StormVisuals/LargeClouds/Mesh_NoLOD/S_SpawnCloudDarkenDisc
  translucencyDirectionalLightingIntensity=0.1
  translucencyLightingMode=TLM_VolumetricPerVertexDirectional
  twoSided=true
  usedShadingModels=Default Lit
 }
 nodes [
  n1 = ParticleColor()
  n2 = ScalarParameter() {defaultValue=1, group=FogFading, parameterName=FogOpacity}
  n3 = TextureCoordinate()
  n4 = TextureSample(UVs=Panner(Coordinate=@n3) {speedY=0.03}) {texture=&T_NoisePack_01}
  n5 = VertexColor()
  n6 = Multiply(A=Divide(A=ComponentMask(None=@n4.RGB) {g=true, r=true}) {constB=15}, B=Multiply(A=@n5.G, B=@n5.G))
  n7 = Add(A=Panner(Coordinate=@n3) {speedY=-0.1}, B=@n6)
  n8 = TextureCoordinate()
  n9 = ComponentMask(None=VectorParameter() {parameterName=TextureB_UVtiling, defaultValue=<rgba 3 1 0 0>}.RGB) {g=true, r=true}
  n10 = Add(A=Panner(Coordinate=Multiply(A=@n8, B=@n9)) {speedY=-0.1}, B=@n6)
  n11 = TextureCoordinate()
  n12 = ComponentMask(None=VectorParameter() {parameterName=TextureC_UVtiling, defaultValue=<rgba 5 1 0 0>}.RGB) {g=true, r=true}
  n13 = Add(A=Panner(Coordinate=Multiply(A=@n11, B=@n12)) {speedY=-0.1}, B=@n6)
  n14 = MaterialFunctionCall(
   A=TextureSample(UVs=@n7) {desc=A, samplerType=SAMPLERTYPE_LinearColor, texture=/Game/Global/Textures/Default/Tileable_Patterns/T_Tile_Smoke_Small}.R
   B=TextureSample(UVs=@n10) {desc=B, samplerType=SAMPLERTYPE_LinearColor, texture=&T_Tile_Smoke_Small}.R
   C=TextureSample(UVs=@n13) {desc=C, texture=/Game/Global/Textures/Default/Tileable_Patterns/T_Smoke_Puff_Tile_Up}.R
   Alpha=MaterialFunctionCall() {materialFunction=&LinearGradient}.VGradient
  ) {materialFunction=&3ColorBlend}
  n15 = OneMinus(None=MaterialFunctionCall() {materialFunction=&LinearGradient}.VGradient)
  n16 = Panner(Coordinate=TextureCoordinate() {uTiling=2}) {speedX=0.2}
  n17 = Saturate(None=Add(A=Multiply(A=@n15) {constB=2.5}, B=TextureSample(UVs=@n16) {texture=&T_Smooth_tile}.R))
  n18 = Multiply(A=Multiply(A=@n14.Result, B=Multiply(A=@n17, B=@n17)), B=VertexColor().G)
  n19 = ShadingPathSwitch(
   Default=DepthFade(Opacity=@n18, FadeDistance=LinearInterpolate(Alpha=@n14.Result) {constA=20, constB=50}) {fadeDistanceDefault=50}
   Mobile=@n18
  ) {desc=Migrated from FeatureLevelSwitch}
  n20 = Subtract(
   A=Multiply(A=Multiply(A=@n1.A, B=@n2), B=QualitySwitch(Default=@n19, Low=@n18, High=@n19, Medium=@n19))
   B=StaticSwitchParameter(
    True=DynamicParameter() {defaultValue=<rgba 0 1 1 1>, paramNames [Erode,Param2,Param3,Param4]}.Erode
    False=ScalarParameter() {group=FogFading, parameterName=FogErode}
   ) {defaultValue=true, group=FogFading, parameterName=UseDynamicParameterToControlErode}
  )
  Clamp(None=@n20, Max=@n20)
  n21 = ReflectionVectorWS(CustomWorldNormal=MaterialFunctionCall(
    A=TextureSample(UVs=@n7) {samplerType=SAMPLERTYPE_Normal, texture=/Game/Effects/Fort_Effects/Creatures/Elemental/Ice/TilingCloudNormalsMap}.RGB
    B=TextureSample(UVs=@n10) {samplerType=SAMPLERTYPE_Normal, texture=&TilingCloudNormalsMap}.RGB
    C=TextureSample(UVs=@n13) {samplerType=SAMPLERTYPE_Normal, texture=&TilingCloudNormalsMap}.RGB
    Alpha=MaterialFunctionCall() {materialFunction=&LinearGradient}.VGradient
   ) {materialFunction=&3ColorBlend}.Result)
  n22 = ShadingPathSwitch(Default=ComponentMask(None=Transform(None=@n21) {transformType=TRANSFORM_View}) {b=true}, Mobile=Constant()) {desc=Migrated from FeatureLevelSwitch}
  n23 = LinearInterpolate(
   A=VectorParameter() {parameterName=ColorLow, defaultValue=<rgba 0.4229 0.366 0.5104 1>}.RGB
   B=VectorParameter() {parameterName=ColorHigh, defaultValue=<rgba 0.7747 0.8735 1 1>}.RGB
   Alpha=@n14.Result
  )
  MaterialFunctionCall(Emissive=Multiply(A=Multiply(A=@n22, B=@n23) {constB=0.2}, B=ScalarParameter() {defaultValue=0.25, parameterName=SpecularAmount}) {constB=0.25}) {materialFunction=&MF_FX_EmissiveAdjustments}
  n24 = AppendVector(A=Constant2Vector(), B=ScalarParameter() {defaultValue=0.25, group=WPO, parameterName=BaseHight})
  n25 = VectorParameter() {group=WPO, parameterName=TextureWPO_UVtiling, defaultValue=<rgba 4 2 0 0>}
  n26 = Panner(Coordinate=Multiply(A=TextureCoordinate(), B=ComponentMask(None=@n25.RGB) {g=true, r=true})) {speedY=-0.15}
  n27 = TextureSample(UVs=@n26) {samplerType=SAMPLERTYPE_Normal, texture=&T_Weap_Distort_N_02}
  n28 = AppendVector(
   A=Multiply(A=ComponentMask(None=@n27.RGB) {g=true, r=true}) {constB=0.5}
   B=Multiply(
    A=TextureSample(UVs=@n26) {texture=&T_Smooth_tile}.R
    B=ScalarParameter() {defaultValue=0.85, group=WPO, parameterName=WPO_Height}
   ) {constB=0.6}
  )
  n29 = Constant() {r=0.5}
  n30 = SphereMask(
   A=Power(
    Base=MaterialFunctionCall() {materialFunction=&LinearGradient}.VGradient
    Exp=ScalarParameter() {defaultValue=0.5, group=WPO, parameterName=CenterOffset}
   ) {constExponent=0.5}
   B=@n29
   Radius=@n29
  ) {hardnessPercent=0}
  n31 = Multiply(A=Transform(None=LinearInterpolate(A=@n24, B=@n28, Alpha=@n30))) {constB=80}
  n32 = Constant3Vector() {desc=Disable WPO anim in Low, constant=<rgba 0 0 20 0>}
  n33 = ShadingPathSwitch(Default=@n31, Deferred=@n31, Mobile=@n32) {desc=Migrated from FeatureLevelSwitch}
  QualitySwitch(Default=@n33, Low=@n32, High=@n33, Medium=@n33)
 ]
}
