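# Niagara system for the smoke spray of a Halloween fog machine prop.
# Three enabled emitters, Smoke, Smoke001 and SmokeLow001, list no renderers.
# Four other emitters, including SmokeLow and underscore-suffixed counterparts of the enabled three, use sprite renderers but are disabled.
# User parameters expose a fog particle spawn rate and a pale blue-white, low-alpha smoke color.
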NiagaraSystem /Game/Environments/Props/Halloween/FX/NS_FogMachine_SmokeSpray
dependencies=null
emitters {Smoke {rendererClasses [], renderers []}, Smoke001 {rendererClasses [], renderers []}, SmokeLow001 {rendererClasses [], renderers []}}
summary {
 systemName=NS_FogMachine_SmokeSpray
 emitters [
  {bEnabled=false, emitterName=Smoke_, rendererClasses [&NiagaraSpriteRendererProperties]}
  {bEnabled=false, emitterName=Smoke001_, rendererClasses [&NiagaraSpriteRendererProperties]}
  {bEnabled=false, emitterName=SmokeLow, rendererClasses [&NiagaraSpriteRendererProperties]}
  {bEnabled=false, emitterName=SmokeLow001_, rendererClasses [&NiagaraSpriteRendererProperties]}
  {emitterName=Smoke, rendererClasses []}
  {emitterName=Smoke001, rendererClasses []}
  {emitterName=SmokeLow001, rendererClasses []}
 ]
 userVariables [
  User.FogParticlesSpawnRate {type=NiagaraFloat, defaultValue=1, description=""}
  User.SmokeColor {type=LinearColor, description="", defaultValue {struct=LinearColor, value=<rgba 0.8747 0.9403 1 0.2>}}
 ]
}
