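# Translucent unlit material for a glowing, animated Halloween ghost prop.
# A world position offset chain spins it around the Z axis at its Pivot Painter per-object pivot using a rotationSpeed parameter, adding a slow sine-driven vertical bob.
# The ghost texture, boosted, tinted cyan-blue and multiplied by Fresnel, forms the glow.
# The texture's red channel, scaled by clamped view-angle fade and Distance_Blend, forms the opacity.
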Material /Game/Environments/Props/Halloween/material/M_Halloween_Ghost_Transparent
graph {
 properties {blendMode=BLEND_Translucent, pixelDepthOffsetMode=PDOM_Legacy, shadingModel=MSM_Unlit, usedShadingModels=Unlit}
 nodes [
  n1 = ScalarParameter() {parameterName=rotationSpeed}
  n2 = Time()
  n3 = MaterialFunctionCall!(Sine Phase=Divide(A=@n2) {constB=9}, Value 1=Constant() {r=-100}, Value 2=Constant() {r=150}) {materialFunction=/Engine/Functions/Engine_MaterialFunctions02/Utility/Sine_Remapped}
  Add(
   A=RotateAboutAxis(
    NormalizedRotationAxis=Constant3Vector() {constant=<rgba 0 0 1 0>}
    RotationAngle=Multiply(A=@n1, B=@n2)
    PivotPoint=MaterialFunctionCall() {materialFunction=/Engine/Functions/Engine_MaterialFunctions02/PivotPainter/PivotPainter_PerObjectData}.Pivot Position
    Position=WorldPosition().XYZ
   ) {period=5}
   B=AppendVector(A=Constant2Vector(), B=ComponentMask(None=@n3.Remapped Sine) {r=true})
  )
  n4 = TextureSample() {texture=/Game/Environments/Props/Halloween/textures/T_Halloween_Ghost_E}
  Multiply(A=Multiply(A=Multiply(A=@n4.RGB) {constB=20}, B=Constant3Vector() {constant=<rgba 0.5417 2.7434 4 0>}), B=Fresnel() {exponent=0.15})
  n5 = MaterialFunctionCall() {materialFunction=&FadeBasedOnViewAngle}
  Multiply(
   A=Multiply(A=@n4.R, B=Clamp(None=@n5.Result) {minDefault=0.3})
   B=MaterialFunctionCall!(Blend Range=Constant() {r=-1250}, Start Offset=Constant() {r=-5000}) {materialFunction=/Engine/Functions/Engine_MaterialFunctions03/Particles/Distance_Blend}.Result
  )
 ]
}
