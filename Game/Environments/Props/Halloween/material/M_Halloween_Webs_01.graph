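# Translucent, two-sided, fully rough material for Halloween spider web props, with a tintable base color and a grayscale web mask as opacity.
# Panning swirl noise, scaled by DistortionStrength (zero by default), offsets the mask UVs to animate the webs.
# Quality and feature-level switches use an undistorted mobile mask on low quality and mobile targets.
# The dither, vertex-alpha gain and contrast branches do not feed the material attributes.
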Material /Game/Environments/Props/Halloween/material/M_Halloween_Webs_01
graph {
 properties {
  bFullyRough=true
  bUseMaterialAttributes=true
  blendMode=BLEND_Translucent
  ditheredLODTransition=true
  twoSided=true
  usedShadingModels=Default Lit
 }
 nodes [
  n1 = TextureSampleParameter2D() {parameterName=Mask_Mobile, samplerType=SAMPLERTYPE_LinearGrayscale, texture=/Game/Environments/Props/Halloween/textures/T_Halloween_Webs_01}
  LinearInterpolate(Alpha=@n1.R)
  n2 = TextureSample(UVs=Panner(Speed=ScalarParameter() {defaultValue=0.1, parameterName=DistortionSpeed})) {samplerType=SAMPLERTYPE_LinearGrayscale, texture=/Game/Global/Textures/Default/Noise/T_SwirlNoise_N_height}
  n3 = ScalarParameter() {parameterName=DistortionStrength}
  n4 = TextureSampleParameter2D(UVs=Add(A=TextureCoordinate(), B=Multiply(A=@n2.R, B=@n3))) {parameterName=MaskTexture, samplerType=SAMPLERTYPE_LinearGrayscale, texture=&T_Halloween_Webs_01}
  n5 = QualitySwitch(Default=@n4.R, Low=@n1.R, High=@n4.R, Medium=@n4.R)
  MakeMaterialAttributes!(
   BaseColor=VectorParameter() {parameterName=Color, defaultValue=<rgba 1 1 1 1>}.RGB
   Metallic
   Specular
   Roughness
   Anisotropy
   EmissiveColor
   Opacity=FeatureLevelSwitch(Default=@n5, ES2=@n1.R, ES3_1=@n1.R, SM4_REMOVED=@n5)
   OpacityMask
   Normal
   Tangent
   WorldPositionOffset
   SubsurfaceColor
   ClearCoat
   ClearCoatRoughness
   AmbientOcclusion
   Refraction
   CustomizedUVs_0
   CustomizedUVs_1
   CustomizedUVs_2
   CustomizedUVs_3
   CustomizedUVs_4
   CustomizedUVs_5
   CustomizedUVs_6
   CustomizedUVs_7
   PixelDepthOffset
   ShadingModel
   Displacement
  )
  n6 = ScalarParameter() {defaultValue=1, parameterName=OpcaityGain}
  MaterialFunctionCall(Alpha Threshold=Multiply(A=VertexColor().A, B=Multiply(A=@n4.R, B=@n6))) {materialFunction=&DitherTemporalAA}
  Power(Base=@n4.RGB, Exp=ScalarParameter() {defaultValue=1, parameterName=OpcaityContrast})
 ]
}
