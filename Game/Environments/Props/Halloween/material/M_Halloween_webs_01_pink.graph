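# Unlit, two-sided alpha-composite material for Halloween spider web props, named as the pink variant, with a tintable Color parameter.
# Panning swirl noise offsets the web mask UVs, with strength defaulting to zero.
# Quality and feature-level switches fall back to an undistorted mobile mask on low-end targets before tinting.
# Separate branches scale the mask by gain and vertex alpha, apply contrast, and add blue-noise dither; output connections are not shown.
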Material /Game/Environments/Props/Halloween/material/M_Halloween_webs_01_pink
graph {
 properties {
  bAutomaticallySetUsageInEditor=false
  bIsTranslucencyVelocityFromDepth=true
  blendMode=BLEND_AlphaComposite
  shadingModel=MSM_Unlit
  translucencyLightingMode=TLM_Surface
  twoSided=true
  usedShadingModels=Unlit
 }
 nodes [
  Constant()
  Constant()
  n1 = TextureSampleParameter2D() {parameterName=Mask_Mobile, samplerType=SAMPLERTYPE_LinearGrayscale, texture=/Game/Environments/Props/Halloween/textures/T_Halloween_Webs_01}
  LinearInterpolate(Alpha=@n1.R)
  n2 = TextureSample(UVs=Panner(Speed=ScalarParameter() {defaultValue=0.1, parameterName=DistortionSpeed})) {samplerType=SAMPLERTYPE_LinearGrayscale, texture=/Game/Global/Textures/Default/Noise/T_SwirlNoise_N_height}
  n3 = ScalarParameter() {parameterName=DistortionStrength}
  n4 = TextureSampleParameter2D(UVs=Add(A=TextureCoordinate(), B=Multiply(A=@n2.R, B=@n3))) {parameterName=MaskTexture, samplerType=SAMPLERTYPE_LinearGrayscale, texture=&T_Halloween_Webs_01}
  Multiply(A=VertexColor().A, B=Multiply(A=@n4.R, B=ScalarParameter() {defaultValue=1, parameterName=OpcaityGain}))
  n5 = QualitySwitch(Default=@n4.R, Low=@n1.R, High=@n4.R, Medium=@n4.R)
  n6 = FeatureLevelSwitch(Default=@n5, ES2=@n1.R, ES3_1=@n1.R, SM4_REMOVED=@n5)
  Multiply(A=VectorParameter() {parameterName=Color, defaultValue=<rgba 1 1 1 1>}.RGB, B=@n6)
  OneMinus(None=@n6)
  Power(Base=@n4.RGB, Exp=ScalarParameter() {defaultValue=1, parameterName=OpcaityContrast})
  Subtract(A=Add(A=ScalarBlueNoise(), B=@n4.R)) {constB=0.5}
 ]
}
